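# Material function that reconstructs the camera-relative world position of the scene behind a translucent surface.
# It divides scene depth by the surface's pixel depth and scales the surface's camera-relative world position by that ratio, projecting it along the view ray.
# An optional scalar input overrides the default SceneDepth lookup for custom sampling.
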MaterialFunction /Game/Global/MaterialFunctions/TechArt/CameraRelativeWorldPositionBehindTranslucency
graph {
 properties {description="This will recreate the scenes world position behind a translucent surface. ", libraryCategoriesText [Misc,Texturing]}
 nodes [
  n1 = FunctionInput(Preview=SceneDepth().R) {
   bUsePreviewValueAsDefault=true
   description=Allows overriding scene depth with custom sampling.
   inputName=SceneDepth
   inputType=FunctionInput_Scalar
  }
  n2 = PixelDepth()
  FunctionOutput(None=Multiply(A=Divide(A=@n1, B=@n2.R), B=WorldPosition() {worldPositionShaderOffset=WPT_CameraRelative}.XYZ))
 ]
}
