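# Library material function that hue-shifts an input color in OKLAB space for perceptually uniform rotation.
# It converts the color through LMS and a cube root to OKLAB, splitting it into lightness, chroma and a fast-arctangent hue angle.
# The shift percentage offsets hue in turns, then sine and cosine rebuild OKLAB for conversion back.
# Its description notes higher instruction cost than a traditional hue shift.
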MaterialFunction /Game/Global/MaterialFunctions/TechArt/MF_HueShift_OKLAB
graph {
 properties {bExposeToLibrary=true}
 nodes [
  n1 = NamedRerouteUsage()
  n2 = NamedRerouteUsage()
  n3 = MaterialFunctionCall(R=NamedRerouteUsage(), G=Multiply(A=Cosine(None=@n1), B=NamedRerouteUsage()), B=Multiply(A=NamedRerouteUsage(), B=Sine(None=@n2))) {materialFunction=&AppendMany}
  n4 = NamedRerouteDeclaration(None=MaterialFunctionCall!(
    VectorToTransform=NamedRerouteDeclaration(None=@n3.RGB) {name=OKLAB_OUT, nodeColor=<rgba 0 1 0.2266 1>}
    BasisX=Constant3Vector() {constant=<rgba 1 0.3963 0.2158 0>}
    BasisY=Constant3Vector() {constant=<rgba 1 -0.1056 -0.0639 0>}
    BasisZ=Constant3Vector() {constant=<rgba 1 -0.0895 -1.2915 0>}
   ) {materialFunction=/Engine/Functions/Engine_MaterialFunctions02/Math/InverseTransformMatrix}.Transformed Vector 3) {name=LMS_OUT, nodeColor=<rgba 1 0 0.0937 1>}
  FunctionOutput(None=MaterialFunctionCall(
    VectorToTransform=Power(Base=@n4) {constExponent=3}
    BasisX=Constant3Vector() {constant=<rgba 4.0767 -3.3077 0.231 0>}
    BasisY=Constant3Vector() {constant=<rgba -1.2684 2.6098 -0.3413 0>}
    BasisZ=Constant3Vector() {constant=<rgba -0.0042 -0.7034 1.7076 0>}
   ) {materialFunction=&InverseTransformMatrix}.Transformed Vector 3) {desc="Provides perceptually uniform hue rotation. Adds ~15/42/38 shader instructions than traditional hueshift on SM6/SM5/Mobile. "}
  n5 = NamedRerouteDeclaration(None=MaterialFunctionCall(
    VectorToTransform=NamedRerouteUsage()
    BasisX=Constant3Vector() {constant=<rgba 0.4122 0.5363 0.0514 0>}
    BasisY=Constant3Vector() {constant=<rgba 0.2119 0.6807 0.1074 0>}
    BasisZ=Constant3Vector() {constant=<rgba 0.0883 0.2817 0.63 0>}
   ) {materialFunction=&InverseTransformMatrix}.Transformed Vector 3) {name=LMS, nodeColor=<rgba 1 0 0.1875 1>}
  n6 = Divide() {constB=3}
  n7 = NamedRerouteDeclaration(None=MaterialFunctionCall(
    VectorToTransform=Power(Base=@n5, Exp=@n6) {constExponent=0.333}
    BasisX=Constant3Vector() {constant=<rgba 0.2105 0.7936 -0.0041 0>}
    BasisY=Constant3Vector() {constant=<rgba 1.978 -2.4286 0.4506 0>}
    BasisZ=Constant3Vector() {constant=<rgba 0.0259 0.7828 -0.8087 0>}
   ) {materialFunction=&InverseTransformMatrix}.Transformed Vector 3) {name=OKLAB, nodeColor=<rgba 1 0.2109 0 1>}
  NamedRerouteDeclaration(None=Length(None=ComponentMask(None=@n7) {b=true, g=true}) {is=class MaterialExpressionLength}) {name=C, nodeColor=<rgba 0 0.1094 1 1>}
  n8 = Arctangent2Fast(Y=ComponentMask(None=@n7) {b=true}, X=ComponentMask(None=@n7) {g=true})
  n9 = MaterialFunctionCall(Multiplier=Constant() {r=2}) {materialFunction=&Pi}
  n10 = Multiply(A=NamedRerouteUsage(), B=@n9.Result)
  NamedRerouteDeclaration(None=Divide(A=Add(A=@n8, B=@n10), B=@n9.Result)) {name=H, nodeColor=<rgba 0.4297 1 0 1>}
  NamedRerouteDeclaration(None=FunctionInput(Preview=Constant3Vector() {constant=<rgba 1 0 0 1>}) {bUsePreviewValueAsDefault=true, inputName=Texture}) {name=InputColor, nodeColor=<rgba 0.125 1 0 1>}
  NamedRerouteDeclaration(None=FunctionInput(Preview=Constant()) {bUsePreviewValueAsDefault=true, inputName=HueShiftPercentage, inputType=FunctionInput_Scalar}) {name=ColorRotationAngle, nodeColor=<rgba 1 0 0.4453 1>}
  NamedRerouteDeclaration(None=ComponentMask(None=@n7) {r=true}) {name=L, nodeColor=<rgba 0.5 0 1 1>}
 ]
}
