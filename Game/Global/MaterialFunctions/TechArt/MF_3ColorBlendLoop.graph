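# Material function in the TechArt library that blends three input colors in a seamless A-B-C-A loop driven by a greyscale Alpha.
# Alpha is scaled by three, then offset by successive subtractions and saturated into one blend weight per segment.
# Three chained lerps apply these weights, the last returning to A so both ends of the range match.
# Preview defaults are red, green and blue over a linear gradient.
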MaterialFunction /Game/Global/MaterialFunctions/TechArt/MF_3ColorBlendLoop
graph {
 properties {
  bExposeToLibrary=true
  description=Blends between 3 colors based on a greyscale input in a loop
  libraryCategoriesText [TechArt, Image Adjustment]
 }
 nodes [
  n1 = FunctionInput() {bUsePreviewValueAsDefault=true, inputName=A, previewValue=<1 0 0 1>}
  n2 = Multiply(A=FunctionInput(Preview=MaterialFunctionCall() {materialFunction=&LinearGradient}.UGradient) {bUsePreviewValueAsDefault=true, inputName=Alpha, inputType=FunctionInput_Scalar, sortPriority=4}) {constB=3}
  n3 = LinearInterpolate(A=@n1, B=FunctionInput() {bUsePreviewValueAsDefault=true, inputName=B, sortPriority=1, previewValue=<0 1 0 1>}, Alpha=Saturate(None=@n2))
  n4 = FunctionInput() {bUsePreviewValueAsDefault=true, inputName=C, sortPriority=1, previewValue=<0 0 1 1>}
  n5 = Subtract(A=@n2)
  n6 = Saturate(None=@n5)
  n7 = Subtract(A=@n5)
  FunctionOutput(None=LinearInterpolate(A=LinearInterpolate(A=@n3, B=@n4, Alpha=@n6), B=@n1, Alpha=Saturate(None=@n7)))
 ]
}
