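# Material function that adds an advanced car clear coat to incoming material attributes, switchable and masked by a selectable channel.
# It sets the ClearCoat shading model, shifts brightened paint toward an alternate colour by powered Fresnel, and varies coat roughness with input roughness and tiled grime.
# Orange-peel detail is angle-blended into the normal.
# Separate outputs expose that normal and a flattened metallic-flake normal (CCN).
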MaterialFunction /Game/Global/MaterialFunctions/TechArt/MF_Advanced_CarClearCoat
graph {nodes [
  Constant3Vector() {constant=<rgba 0.5 0.5 1 0>}
  Constant()
  n1 = FunctionInput() {inputType=FunctionInput_MaterialAttributes}
  n2 = ScalarParameter() {defaultValue=2, parameterName=ClearCoatBaseColorBrightness}
  n3 = GetMaterialAttributes(None=@n1) {attributeGetTypes [#1, #2, #3]}
  n4 = VectorParameter() {group=06_AdvancedClearCoat, parameterName=ClearCoatAlternatePaintColor, defaultValue=<rgba 0.2188 0.3082 1 1>}
  n5 = Power(
   Base=Abs(None=FunctionInput() {inputName=Fresnel, inputType=FunctionInput_Scalar})
   Exp=ScalarParameter() {defaultValue=2, group=06_AdvancedClearCoat, parameterName=ExpPaintFresnel}
  ) {constExponent=5}
  n6 = ScalarParameter() {defaultValue=0.01, group=06_AdvancedClearCoat, parameterName=Clear Coat Roughness}
  n7 = ScalarParameter() {defaultValue=0.3, group=06_AdvancedClearCoat, parameterName=Grime Roughness}
  n8 = Multiply(A=TextureCoordinate(), B=ScalarParameter() {defaultValue=6, group=06_AdvancedClearCoat, parameterName=GrimeTiling})
  n9 = LinearInterpolate(A=@n3.Roughness, B=TextureSample(UVs=@n8) {samplerType=SAMPLERTYPE_Masks, texture=/Game/Global/Textures/Default/Noise/T_Tiling_Grime}.B)
  n10 = TextureSample!(UVs=Multiply(A=@n8, B=Constant() {r=2}) {constB=2}, Tex, Bias=Constant() {r=1}, Apply View MipBias) {mipValueMode=TMVM_MipBias, samplerType=SAMPLERTYPE_Normal, texture=/Game/Global/Textures/Default/Noise/T_OrangePeel_N}
  n11 = LinearInterpolate(Alpha=ScalarParameter() {defaultValue=0.1, group=06_AdvancedClearCoat, parameterName=OrangePeelIntensity}) {constA=1, constB=0}
  n12 = Normalize(VectorInput=MaterialFunctionCall(
    BaseNormal=@n3.Normal
    AdditionalNormal=MaterialFunctionCall(Normal=@n10.RGB, Flatness=@n11) {materialFunction=&FlattenNormal}.Result
   ) {materialFunction=&BlendAngleCorrectedNormals}.Result)
  n13 = SetMaterialAttributes!(
   MaterialAttributes=@n1
   BaseColor=LinearInterpolate(A=Multiply(A=@n2, B=@n3.BaseColor), B=@n4.RGB, Alpha=@n5)
   Metallic=ScalarParameter() {defaultValue=1, group=06_AdvancedClearCoat, parameterName=ClearCoatMetallic}
   Specular=Constant() {r=1}
   Roughness=ScalarParameter() {defaultValue=0.5, group=06_AdvancedClearCoat, parameterName=BaseRoughnessMult}
   Anisotropy
   CustomData0=Constant() {r=1}
   CustomData1=Saturate(None=LinearInterpolate(A=@n6, B=@n7, Alpha=@n9))
   ShadingModel=ShadingModel() {shadingModel=MSM_ClearCoat}
   Normal=@n12
  ) {attributeSetTypes [
    #2
    #4
    #5
    #1
    #6
    #7
    #8
    #9
    #3
   ]}
  n14 = ChannelMaskParameter(None=FunctionInput() {inputName=Mask}) {group=05_Tech Art Masks, parameterName=AdvancedClearCoatChannel}
  FunctionOutput(None=StaticSwitchParameter(True=BlendMaterialAttributes(A=@n1, B=@n13, Alpha=@n14.R), False=@n1) {defaultValue=true, group=04_Tech Art, parameterName=UseAdvancedClearCoat})
  n15 = Multiply(A=TextureCoordinate(), B=ScalarParameter() {defaultValue=60, group=06_AdvancedClearCoat, parameterName=FlakeTiling})
  n16 = ScalarParameter() {defaultValue=0.5, group=06_AdvancedClearCoat, parameterName=ClearCoatFlakeIntensity}
  FunctionOutput(None=MaterialFunctionCall(
    Normal=TextureSample!(UVs=@n15, Tex, Apply View MipBias) {samplerType=SAMPLERTYPE_Normal, texture=/Game/Environments/Sets/Neo/Vehicles/Textures/T_Flakes_N}.RGB
    Flatness=LinearInterpolate(Alpha=@n16) {constA=1, constB=0}
   ) {materialFunction=&FlattenNormal}.Result) {outputName=CCN}
  FunctionOutput(None=@n12) {outputName=NormalWithOrangePeel}
  LinearInterpolate()
  NamedRerouteDeclaration(None=@n5) {name=Fresnel, nodeColor=<rgba 0.3125 0 1 1>}
 ]}
