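# Material function that adds an optional car-paint clear coat layer to incoming material attributes.
# It switches to the ClearCoat shading model, tints brightened base color with Fresnel, derives coat roughness from a tiled grime texture, and can override metallic, specular and roughness.
# Channel-selectable masks drive the layer blend and optional normal flattening.
# A separate ClearCoatNormal output blends a tiled flake normal with a base normal using angle-corrected blending.
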MaterialFunction /Game/Global/MaterialFunctions/TechArt/MF_Advanced_CarClearCoat_v2
graph {nodes [
  n1 = FunctionInput() {inputType=FunctionInput_MaterialAttributes}
  n2 = GetMaterialAttributes(None=@n1) {attributeGetTypes [
    #1
    #2
    #3
    #4
    #5
   ]}
  n3 = SetMaterialAttributes!(
   MaterialAttributes=@n2.MaterialAttributes
   BaseColor=NamedRerouteUsage()
   Metallic=StaticSwitch(
    True=ScalarParameter() {defaultValue=1, group=06_AdvancedClearCoat_v2, parameterName=ClearCoat_OverrideMetallic, sortPriority=4}
    False=@n2.Metallic
    Value=NamedRerouteUsage()
   )
   Specular=StaticSwitch(
    True=ScalarParameter() {defaultValue=1, group=06_AdvancedClearCoat_v2, parameterName=ClearCoat_OverrideSpecular, sortPriority=4}
    False=@n2.Specular
    Value=NamedRerouteUsage()
   )
   Roughness=StaticSwitch(
    True=ScalarParameter() {defaultValue=0.5, group=06_AdvancedClearCoat_v2, parameterName=ClearCoat_OverrideRoughness, sortPriority=4}
    False=@n2.Roughness
    Value=NamedRerouteUsage()
   )
   CustomData0=ScalarParameter() {defaultValue=1, group=06_AdvancedClearCoat_v2, parameterName=ClearCoat_Strength}
   CustomData1=NamedRerouteUsage()
   ShadingModel=ShadingModel() {shadingModel=MSM_ClearCoat}
  ) {attributeSetTypes [
    #2
    #4
    #5
    #1
    #6
    #7
    #8
   ]}
  n4 = NamedRerouteUsage()
  FunctionOutput(None=StaticSwitchParameter(True=BlendMaterialAttributes(A=@n1, B=@n3, Alpha=@n4), False=@n1) {group=04_Tech Art, parameterName=UseAdvancedClearCoat_v2})
  n5 = TextureSampleParameter2D(UVs=Multiply(
    A=NamedRerouteUsage()
    B=ScalarParameter() {defaultValue=10, group=06_AdvancedClearCoat_v2, parameterName=ClearCoat_NormalTiling, sortPriority=14}
   )) {
   group=06_AdvancedClearCoat_v2
   parameterName=ClearCoat_Normal
   samplerType=SAMPLERTYPE_Normal
   sortPriority=13
   texture=/Game/Environments/Sets/Neo/Vehicles/Textures/T_Flakes_N
  }
  n6 = MaterialFunctionCall(
   Normal=@n5.RGB
   Flatness=LinearInterpolate(Alpha=ScalarParameter() {defaultValue=0.5, group=06_AdvancedClearCoat_v2, parameterName=ClearCoat_NormalIntensity, sortPriority=15}) {constA=1, constB=0}
  ) {materialFunction=&FlattenNormal}
  n7 = Constant3Vector() {constant=<rgba 0 0 1 0>}
  n8 = NamedRerouteUsage()
  FunctionOutput(None=MaterialFunctionCall(BaseNormal=NamedRerouteUsage(), AdditionalNormal=LinearInterpolate(A=@n6.Result, B=@n7, Alpha=@n8)) {materialFunction=&BlendAngleCorrectedNormals}.Result) {outputName=ClearCoatNormal}
  n9 = FunctionInput() {inputName=Fresnel, inputType=FunctionInput_Scalar}
  NamedRerouteDeclaration(None=Power(
    Base=Abs(None=@n9)
    Exp=ScalarParameter() {defaultValue=2, group=06_AdvancedClearCoat_v2, parameterName=ClearCoat_FresnelPower, sortPriority=5}
   ) {constExponent=5}) {name=Fresnel, nodeColor=<rgba 0.3125 0 1 1>}
  NamedRerouteDeclaration(None=StaticSwitchParameter(True=TextureCoordinate() {coordinateIndex=1}, False=TextureCoordinate()) {group=06_AdvancedClearCoat_v2, parameterName=ClearCoat_UseUV2, sortPriority=-1}) {name=UVs, nodeColor=<rgba 1 0.2578 0 1>}
  n10 = ScalarParameter() {group=06_AdvancedClearCoat_v2, parameterName=ClearCoat_RoughnessMin, sortPriority=10}
  n11 = ScalarParameter() {defaultValue=1, group=06_AdvancedClearCoat_v2, parameterName=ClearCoat_RoughnessMax, sortPriority=11}
  n12 = TextureSampleParameter2D(UVs=Multiply(
    A=NamedRerouteUsage()
    B=ScalarParameter() {defaultValue=5, group=06_AdvancedClearCoat_v2, parameterName=ClearCoat_RoughnessTiling, sortPriority=9}
   )) {
   group=06_AdvancedClearCoat_v2
   parameterName=ClearCoat_Roughness
   samplerType=SAMPLERTYPE_Masks
   sortPriority=7
   texture=/Game/Global/Textures/Default/Noise/T_Tiling_Grime
  }
  n13 = ChannelMaskParameter(None=@n12.RGB) {group=06_AdvancedClearCoat_v2, maskChannel=Blue, parameterName=ClearCoat_RoughnessChannel, sortPriority=8, defaultValue=<rgba 0 0 1 0>}
  n14 = LinearInterpolate(
   A=@n13.R
   B=NamedRerouteUsage()
   Alpha=ScalarParameter() {group=06_AdvancedClearCoat_v2, parameterName=ClearCoat_BaseRoughnessInfluence, sortPriority=12}
  )
  NamedRerouteDeclaration(None=Saturate(None=LinearInterpolate(A=@n10, B=@n11, Alpha=@n14))) {name=ClearCoatRoughness, nodeColor=<rgba 0 0.0625 1 1>}
  n15 = ScalarParameter() {defaultValue=2, group=06_AdvancedClearCoat_v2, parameterName=ClearCoat_BaseColorBrightness, sortPriority=6}
  n16 = NamedRerouteUsage()
  n17 = VectorParameter() {group=06_AdvancedClearCoat_v2, parameterName=ClearCoat_FresnelColor, sortPriority=5, defaultValue=<rgba 0.2188 0.3082 1 1>}
  n18 = NamedRerouteUsage()
  n19 = ScalarParameter() {defaultValue=1, group=06_AdvancedClearCoat_v2, parameterName=ClearCoat_BaseColorFresnelStrength, sortPriority=6}
  NamedRerouteDeclaration(None=LinearInterpolate(A=Multiply(A=@n15, B=@n16), B=@n17.RGB, Alpha=Multiply(A=@n18, B=@n19))) {name=ClearCoatBaseColor, nodeColor=<rgba 0.3828 1 0 1>}
  NamedRerouteDeclaration(None=@n2.BaseColor) {name=BaseColor_IN, nodeColor=<rgba 1 0.4922 0 1>}
  NamedRerouteDeclaration(None=@n2.Roughness) {name=Roughness_IN, nodeColor=<rgba 0.1719 0 1 1>}
  n20 = FunctionInput() {inputName=Mask}
  NamedRerouteDeclaration(None=ChannelMaskParameter(None=@n20) {group=05_Tech Art Masks, parameterName=AdvancedClearCoatChannel, sortPriority=2}.R) {name=ClearCoatChannel, nodeColor=<rgba 0.1484 1 0 1>}
  n21 = ChannelMaskParameter(None=@n20) {group=05_Tech Art Masks, maskChannel=Green, parameterName=AdvancedClearCoatNormalMaskChannel, sortPriority=2, defaultValue=<rgba 0 1 0 0>}
  NamedRerouteDeclaration(None=StaticSwitchParameter(True=@n21.R, False=Constant()) {group=06_AdvancedClearCoat_v2, parameterName=ClearCoat_UseNormalMask, sortPriority=0}) {name=ClearCoatNormalMask, nodeColor=<rgba 1 0 0.4687 1>}
  NamedRerouteDeclaration(None=@n2.Normal) {name=Normal_IN, nodeColor=<rgba 1 0 0.7266 1>}
  NamedRerouteDeclaration(None=StaticBoolParameter() {group=06_AdvancedClearCoat_v2, parameterName=ClearCoat_OverrideSRM, sortPriority=1}) {name=OverrideSRM, nodeColor=<rgba 0.3828 1 0 1>}
 ]}
