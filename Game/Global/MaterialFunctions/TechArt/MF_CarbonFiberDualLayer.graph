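# Material function that layers a dual-layer, clear-coat carbon fiber finish onto incoming material attributes within a channel-selected mask.
# A tiled carbon fiber texture supplies tinted base color, while its contrast-adjusted alpha drives undercoat roughness.
# A flattened orange-peel normal blends angle-corrected into the incoming normal, and the incoming roughness, scaled and clamped, becomes clear-coat roughness.
# It outputs the blended attributes, the carbon fiber UVs and the mask.
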MaterialFunction /Game/Global/MaterialFunctions/TechArt/MF_CarbonFiberDualLayer
graph {nodes [
  n1 = FunctionInput() {inputType=FunctionInput_MaterialAttributes}
  n2 = VectorParameter() {group=Dual Layer Carbon Fiber, parameterName=CarbonFiberTint, defaultValue=<rgba 0.1562 0.1562 0.1562 1>}
  n3 = Multiply(A=TextureCoordinate(), B=ScalarParameter() {defaultValue=8, group=Dual Layer Carbon Fiber, parameterName=CarbonFiberTiling})
  n4 = TextureSample(UVs=@n3) {texture=/Game/Weapons/WeaponSkins/Wraps/Textures/T_Wrap_CarbonFiber_D}
  n5 = MaterialFunctionCall(
   In=Multiply(A=@n4.A) {constB=1.6}
   Contrast=ScalarParameter() {defaultValue=0.5, group=Dual Layer Carbon Fiber, parameterName=contrastroughness}
  ) {materialFunction=&CheapContrast}
  n6 = GetMaterialAttributes(None=@n1) {attributeGetTypes [#1, #2, #3]}
  n7 = TextureSample(UVs=Multiply(A=@n3, B=ScalarParameter() {defaultValue=0.5, group=Dual Layer Carbon Fiber, parameterName=OrangePeelTiling})) {samplerType=SAMPLERTYPE_Normal, texture=/Game/Global/Textures/Default/Noise/T_OrangePeel_N}
  n8 = LinearInterpolate(Alpha=ScalarParameter() {defaultValue=0.05, group=Dual Layer Carbon Fiber, parameterName=OrangePeelIntensity}) {constA=1, constB=0}
  n9 = ScalarParameter() {defaultValue=1, group=Dual Layer Carbon Fiber, parameterName=ClearCoatRoughnessMultiplier}
  n10 = SetMaterialAttributes!(
   MaterialAttributes=@n1
   BaseColor=Multiply(A=@n2.RGB, B=@n4.RGB)
   Metallic=Constant()
   Specular=Constant() {r=0.4}
   Roughness=LinearInterpolate(
    A=ScalarParameter() {defaultValue=0.4, group=Dual Layer Carbon Fiber, parameterName=UndercoatRoughness_Min}
    B=ScalarParameter() {defaultValue=0.6, group=Dual Layer Carbon Fiber, parameterName=UndercoatRoughness_Max}
    Alpha=Saturate(None=@n5.Result)
   )
   Normal=MaterialFunctionCall(
    BaseNormal=@n6.Normal
    AdditionalNormal=MaterialFunctionCall(Normal=@n7.RGB, Flatness=@n8) {materialFunction=&FlattenNormal}.Result
   ) {materialFunction=&BlendAngleCorrectedNormals}.Result
   ShadingModel=ShadingModel() {shadingModel=MSM_ClearCoat}
   CustomData0=Constant() {r=1}
   CustomData1=Clamp(None=Multiply(A=@n6.Roughness, B=@n9))
   Opacity=Constant()
  ) {attributeSetTypes [
    #3
    #4
    #5
    #1
    #2
    #6
    #7
    #8
    #9
   ]}
  n11 = ChannelMaskParameter(None=FunctionInput() {inputName=Alpha}) {parameterName=CarbonFiberChannel}
  FunctionOutput(None=BlendMaterialAttributes(A=@n1, B=@n10, Alpha=@n11.R))
  FunctionOutput(None=@n3) {outputName=CarbonFiberUvs}
  FunctionOutput(None=@n11.R) {outputName=Mask}
  TextureCoordinate()
 ]}
