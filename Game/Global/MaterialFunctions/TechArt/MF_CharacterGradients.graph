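# Material function that builds two character gradient masks, ArmsToFeetGradient and RadialGradient, from pre-skinned local position and reference-pose landmarks.
# ArmsToFeetGradient blends a throat-to-feet height projection with a fingertip-to-shoulder projection via the Arms Mask, mirroring arms by absolute X and feeding local Y into both Y and Z.
# RadialGradient swaps the height term for a bounds-based, zero-hardness sphere mask.
# Static switches replace rump or arm gradients with constants.
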MaterialFunction /Game/Global/MaterialFunctions/TechArt/MF_CharacterGradients
graph {nodes [
  n1 = Constant() {r=1}
  n2 = FunctionInput(Preview=VertexInterpolator(VS=PreSkinnedPosition()).PS) {bUsePreviewValueAsDefault=true, description=Defaults to Pre-Skinned Local Position, inputName=Local Position, sortPriority=-5}
  n3 = MaterialFunctionCall(Float3=@n2) {materialFunction=&BreakOutFloat3Components}
  n4 = FunctionInput() {
   bUsePreviewValueAsDefault=true
   description="Local Z position of throat/clavicle in character ref pose, used as the start of the rump gradient. NOTE: The gradient will mirror at this point."
   inputName=Rump_GradientBeginHeight
   inputType=FunctionInput_Scalar
   sortPriority=100
   previewValue=<140 8 140 1>
  }
  n5 = MaterialFunctionCall(Z=@n4) {materialFunction=&MakeFloat3}
  n6 = FunctionInput() {
   bUsePreviewValueAsDefault=true
   description="Local Z position of bottom of feet in character ref pose, used as the end of the rump gradient"
   inputName=Rump_GradientEndHeight
   inputType=FunctionInput_Scalar
   sortPriority=100
   previewValue=<0 8 0 1>
  }
  n7 = Divide(A=DotProduct(A=Constant3Vector() {constant=<rgba 0 0 -1 0>}, B=Subtract(A=@n3.B, B=@n5.Result)), B=Abs(None=Subtract(A=@n4, B=@n6)))
  n8 = OneMinus(None=Saturate(None=Abs(None=@n7)))
  n9 = FunctionInput() {
   bUsePreviewValueAsDefault=true
   description=Scales the Rump gradient from the end. Adjust this to create a good transition to the Arm gradient.
   inputName=RumpBalance
   inputType=FunctionInput_Scalar
   sortPriority=1
   previewValue=<0.5 0 0 1>
  }
  n10 = OneMinus(None=Multiply(
    A=NamedRerouteDeclaration(None=@n8) {name=RumpGradientRaw, nodeColor=<rgba 0.8984 0 1 1>}
    B=NamedRerouteDeclaration(None=@n9) {name=RumpBalanceInput, nodeColor=<rgba 0.4062 1 0 1>}
   ))
  n11 = StaticBool()
  n12 = FunctionInput(Preview=@n11) {bUsePreviewValueAsDefault=true, inputName=Disable Rump Gradient, inputType=FunctionInput_StaticBool, sortPriority=301}
  n13 = Constant()
  n14 = MaterialFunctionCall(X=Abs(None=@n3.R), Y=@n3.G, Z=@n3.G) {materialFunction=&MakeFloat3}
  n15 = FunctionInput() {
   bUsePreviewValueAsDefault=true
   description="Local position of fingertips in character ref pose, used as the start of the arm gradients. Note: The gradient will be mirrored at this point, so ensure it starts at the very fingertips."
   inputName=Arm_GradientBeginPosition
   sortPriority=90
   previewValue=<60 -6 88 1>
  }
  n16 = FunctionInput() {
   bUsePreviewValueAsDefault=true
   description="Direction Vector from hand to shoulder/clavicle in character ref pose. This will be mirrored, so make sure to use"
   inputName=Arm_GradientDirection
   sortPriority=90
   previewValue=<-1 -0.15 0.75 1>
  }
  n17 = FunctionInput() {
   bUsePreviewValueAsDefault=true
   description="Local position of shoulder/clavicle/throat in character ref pose, used as the end of the arm gradients"
   inputName=Arm_GradientEndPosition
   sortPriority=90
   previewValue=<0 8 150 1>
  }
  n18 = Divide(A=DotProduct(A=Subtract(A=@n14.Result, B=@n15), B=Normalize(VectorInput=@n16)), B=Distance(A=@n15, B=@n17))
  n19 = FunctionInput() {
   bUsePreviewValueAsDefault=true
   description=Scales the Arm gradient from the start. Adjust this to create a good transition to the Rump gradient.
   inputName=ArmBalance
   inputType=FunctionInput_Scalar
   previewValue=<0.5 0 0 1>
  }
  n20 = Multiply(
   A=NamedRerouteDeclaration(None=Saturate(None=@n18)) {name=ArmGradientRaw, nodeColor=<rgba 0 0.8828 1 1>}
   B=NamedRerouteDeclaration(None=@n19) {name=ArmBalanceInput, nodeColor=<rgba 0 0.4609 1 1>}
  ) {constB=0.7}
  n21 = FunctionInput(Preview=@n11) {bUsePreviewValueAsDefault=true, inputName=Disable Arm Gradients, inputType=FunctionInput_StaticBool, sortPriority=300}
  n22 = FunctionInput() {
   description="Expects a Mask that is white on the arms and black on the rump, with a soft falloff gradient fade in between."
   inputName=Arms Mask
   inputType=FunctionInput_Scalar
   sortPriority=10
  }
  FunctionOutput(None=LinearInterpolate(A=StaticSwitch(True=@n1, False=@n10, Value=@n12), B=StaticSwitch(True=@n13, False=@n20, Value=@n21), Alpha=@n22)) {outputName=ArmsToFeetGradient}
  n23 = Constant() {r=1}
  n24 = FunctionInput() {
   bUsePreviewValueAsDefault=true
   description=Used to adjust the horizontal scale of the radial gradient. A value of 1.0 should be circular based on bounds
   inputName=Radial_HorizontalCompression
   inputType=FunctionInput_Scalar
   sortPriority=200
   previewValue=<1 0 0 1>
  }
  n25 = PreSkinnedLocalBounds()
  n26 = ComponentMask(None=@n25.Extents) {r=true}
  n27 = MaterialFunctionCall(Float3=@n25.Half Extents) {materialFunction=&BreakOutFloat3Components}
  n28 = Divide(A=@n27.R, B=@n27.R)
  n29 = Divide(A=@n3.R, B=Multiply(A=@n24, B=Divide(A=@n26, B=@n28)))
  n30 = NamedRerouteUsage()
  n31 = Constant()
  n32 = FunctionInput() {
   bUsePreviewValueAsDefault=true
   description="Sets the normalized (0...1 based on bounds) z-height of the radial gradient's origin. Use in conjunction with Radial_Scale and Radial_HorizontalCompression to get a good transition to the arms, and cover the parts of the rump that the character needs."
   inputName=Radial_CenterHeight
   inputType=FunctionInput_Scalar
   sortPriority=200
   previewValue=<0.5 0 0 1>
  }
  n33 = OneMinus(None=SphereMask(A=AppendVector(A=@n29, B=@n30), B=AppendVector(A=@n31, B=@n32), Radius=NamedRerouteUsage()) {hardnessPercent=0})
  n34 = Constant()
  n35 = Multiply(A=OneMinus(None=NamedRerouteUsage()), B=NamedRerouteUsage()) {constB=0.7}
  FunctionOutput(None=LinearInterpolate(A=StaticSwitch(True=@n23, False=@n33, Value=@n12), B=StaticSwitch(True=@n34, False=@n35, Value=@n21), Alpha=@n22)) {outputName=RadialGradient}
 ]}
