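# Exposed scalar material function that selects one of five platform-tier values by shading path and quality level.
# Deferred and default paths use Low (PC Low/Med), High (XB1, PS4) or Epic (high-end PC/consoles), with unwired Medium falling back to Epic.
# Mobile uses Mobile Low (PC Perf, Android) for Low, otherwise Mobile Med (Switch).
# An unconnected dashed input serves as a pin separator.
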MaterialFunction /Game/Global/MaterialFunctions/TechArt/MF_Fortnite_QualitySwitch_Scalar
graph {
 properties {bExposeToLibrary=true}
 nodes [
  FunctionInput() {bUsePreviewValueAsDefault=true, inputName=--------, inputType=FunctionInput_Scalar, sortPriority=3}
  n1 = FunctionInput() {
   bUsePreviewValueAsDefault=true
   desc="PC-High && Epic, XSX, PS5"
   description="PC-High && Epic, XSX, PS5"
   inputName="Epic _ PC High, XSX, PS5"
   inputType=FunctionInput_Scalar
   sortPriority=2
  }
  n2 = FunctionInput() {
   bUsePreviewValueAsDefault=true
   desc="PC Low && Medium (not Perf mode)"
   description="PC Low && Medium (not PerfMode)"
   inputName=Low _ PC Low and Med
   inputType=FunctionInput_Scalar
  }
  n3 = FunctionInput() {bUsePreviewValueAsDefault=true, desc="XB1, PS4\r\nno SSColor", inputName="High _ XB1, PS4", inputType=FunctionInput_Scalar, sortPriority=1}
  n4 = QualitySwitch(Default=@n1, Low=@n2, High=@n3, Epic=@n1) {desc=has GBuffers}
  n5 = FunctionInput() {
   bUsePreviewValueAsDefault=true
   desc="Switch\r\nhas outlines?\r\nno roughness outlines"
   description=Switch
   inputName=Mobile Med _ Switch
   inputType=FunctionInput_Scalar
   sortPriority=5
  }
  n6 = FunctionInput() {
   bUsePreviewValueAsDefault=true
   desc="PC_PerfMode, Android\r\nno outlines (depth buffer cant be transformed)\r\nno roughness outlines"
   description="PC_PerfMode, Android"
   inputName="Mobile Low _ PC Perf, Android"
   inputType=FunctionInput_Scalar
   sortPriority=4
  }
  FunctionOutput(None=ShadingPathSwitch(Default=@n4, Deferred=@n4, Mobile=QualitySwitch(Default=@n5, Low=@n6, Medium=@n5) {desc="no GBuffers, has depth"}))
 ]
}
