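# Material function that adds a holographic glitter layer to input material attributes, toggled by the isGlitter static switch and masked by a selectable channel of Alpha.
# A Fresnel term, distorted by animated Perlin noise, indexes a thin-film spectrum texture added to base color.
# Metallic-flake normals, flattened by Fresnel, blend into the surface normal.
# A static switch selects triplanar or UV sampling; a second thin-film branch is left unconnected.
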MaterialFunction /Game/Global/MaterialFunctions/TechArt/MF_GlitterHolographic
graph {nodes [
  n1 = FunctionInput() {inputType=FunctionInput_MaterialAttributes}
  n2 = GetMaterialAttributes(None=@n1) {attributeGetTypes [#1, #2, #3]}
  n3 = Fresnel(ExponentIn=ScalarParameter() {defaultValue=0.8, group=06_Glitter, parameterName=GlitterFalloffRadius}, Normal=VertexNormalWS()) {baseReflectFraction=0}
  n4 = Power(Base=@n3, Exp=ScalarParameter() {defaultValue=1, group=06_Glitter, parameterName=GlitterFilmPower})
  n5 = ScalarParameter() {defaultValue=1, group=06_Glitter, parameterName=GlitterFilmTiling}
  n6 = TextureObject() {samplerType=SAMPLERTYPE_Masks, texture=/Game/Global/Textures/Default/Noise/T_Generic_Perlin}
  n7 = NamedRerouteUsage()
  n8 = NamedRerouteUsage()
  n9 = NamedRerouteUsage()
  n10 = StaticBoolParameter() {group=06_Glitter, parameterName=GlitterUseTriplaner}
  n11 = MaterialFunctionCall(Float3=StaticSwitch(
    True=MaterialFunctionCall(TextureObject=@n6, TextureSize=@n7, WorldPosition=@n8, ProjectionTransitionContrast=@n9) {materialFunction=&WorldAlignedTexture}.XYZ Texture
    False=TextureSample() {samplerType=SAMPLERTYPE_Masks, texture=&T_Generic_Perlin}.RGB
    Value=@n10
   )) {materialFunction=&BreakOutFloat3Components}
  n12 = NamedRerouteUsage()
  n13 = Time()
  n14 = Add(
   A=Multiply(A=Frac(None=@n13), B=VectorParameter() {group=06_Glitter, parameterName=TriplannerFlamePanning, defaultValue=<rgba 0.1 0 0 1>}.RGB)
   B=NamedRerouteUsage()
  )
  n15 = NamedRerouteUsage()
  n16 = Multiply(A=MaterialFunctionCall(X=@n11.G, Y=@n11.G) {materialFunction=&MakeFloat2}.Result) {constB=0.1}
  n17 = Panner(Coordinate=Multiply(A=TextureCoordinate(), B=ScalarParameter() {defaultValue=0.5, parameterName=GlitterFlameTiling})) {speedY=0.02}
  n18 = StaticSwitch(
   True=ComponentMask(None=MaterialFunctionCall(TextureObject=@n6, TextureSize=@n12, WorldPosition=@n14, ProjectionTransitionContrast=@n15) {materialFunction=&WorldAlignedTexture}.XYZ Texture) {g=true}
   False=TextureSample(UVs=Add(A=@n16, B=@n17)) {samplerType=SAMPLERTYPE_Masks, texture=&T_Generic_Perlin}.G
   Value=@n10
  )
  n19 = TextureSample(UVs=Add(A=Multiply(A=@n4, B=@n5), B=Multiply(A=@n11.R, B=@n18))) {texture=/Game/Global/Textures/Default/Thinfilm/T_ThinFilm_Spectrum_COLOR}
  n20 = ScalarParameter() {defaultValue=0.2, group=06_Glitter, parameterName=GlitterThinfilmOpacity}
  n21 = Add(A=@n2.BaseColor, B=Multiply(A=@n3, B=Multiply(A=@n19.RGB, B=@n20)))
  n22 = ScalarParameter() {defaultValue=1, group=06_Glitter, parameterName=ThinFilmScaleGlitter, sortPriority=5}
  n23 = FunctionInput() {inputName=Fresnel, inputType=FunctionInput_Scalar, sortPriority=2}
  n24 = Power(Base=@n23, Exp=ScalarParameter() {defaultValue=0.5, group=06_Glitter, parameterName=ThinFilmWarpGlitter, sortPriority=4})
  n25 = TextureSampleParameter2D(UVs=Add(A=Multiply(A=@n22, B=@n24), B=ScalarParameter() {group=06_Glitter, parameterName=ThinFilmOffsetGlitter})) {
   group=06_Thin Film
   parameterName=Thin Film Texture
   sortPriority=1
   texture=&T_ThinFilm_Spectrum_COLOR
  }
  n26 = Multiply(A=@n25.RGB, B=ScalarParameter() {defaultValue=0.5, group=06_Glitter, parameterName=ThinFilmMultiplyGlitter, sortPriority=2})
  n27 = Power(Base=@n23, Exp=ScalarParameter() {defaultValue=2, group=06_Glitter, parameterName=ThinFilmPowerGlitter, sortPriority=3})
  n28 = LinearInterpolate(
   A=@n27
   B=Multiply(A=LinearInterpolate(Alpha=@n2.Roughness) {constA=1, constB=0.01}, B=@n27)
   Alpha=ScalarParameter() {group=06_Glitter, parameterName=ThinFilmIntensityGlitter}
  )
  Add(A=@n21, B=LinearInterpolate(A=@n2.BaseColor, B=Add(A=@n2.BaseColor, B=@n26), Alpha=Saturate(None=@n28)))
  n29 = TextureObject() {samplerType=SAMPLERTYPE_Normal, texture=/Game/Weapons/WeaponSkins/Wraps/Textures/T_Wrap_metallicFlakes_Small}
  n30 = TextureCoordinate()
  n31 = ScalarParameter() {defaultValue=4, group=06_Glitter, parameterName=GlitterTiling}
  n32 = StaticSwitch(
   True=MaterialFunctionCall(TextureObject=@n29, TextureSize=NamedRerouteUsage(), WorldPosition=NamedRerouteUsage(), ProjectionTransitionContrast=NamedRerouteUsage()) {materialFunction=&WorldAlignedTexture}.XYZ Texture
   False=TextureSample(UVs=Multiply(A=@n30, B=@n31), Tex=@n29) {samplerType=SAMPLERTYPE_Normal, texture=&T_Wrap_metallicFlakes_Small}.RGB
   Value=@n10
  )
  n33 = LinearInterpolate(A=ScalarParameter() {defaultValue=0.1, group=06_Glitter, parameterName=GlitterIntensity}, Alpha=@n3)
  n34 = SetMaterialAttributes!(
   MaterialAttributes=@n2.MaterialAttributes
   Normal=MaterialFunctionCall(
    BaseNormal=@n2.Normal
    AdditionalNormal=MaterialFunctionCall(Normal=@n32, Flatness=@n33) {materialFunction=&FlattenNormal}.Result
   ) {materialFunction=&BlendAngleCorrectedNormals}.Result
   BaseColor=@n21
  ) {attributeSetTypes [#1, #2]}
  n35 = ChannelMaskParameter(None=FunctionInput() {inputName=Alpha}) {parameterName=GlitterMaskChannel}
  FunctionOutput(None=StaticSwitchParameter(True=BlendMaterialAttributes(A=@n2.MaterialAttributes, B=@n34, Alpha=@n35.R), False=@n1) {parameterName=isGlitter})
  MaterialFunctionCall(Float3=@n18) {materialFunction=&BreakOutFloat3Components}
  NamedRerouteDeclaration(None=ScalarParameter() {defaultValue=64, group=06_Glitter, parameterName=GlitterSize}) {name=GlitterSize, nodeColor=<rgba 1 0 0.2578 1>}
  NamedRerouteDeclaration(None=ScalarParameter() {defaultValue=0.2, group=06_Glitter, parameterName=GlitterTransitionContrast}) {name=GlitterTransitionContrast, nodeColor=<rgba 1 0 0.1406 1>}
  NamedRerouteDeclaration(None=VertexInterpolator(VS=PreSkinnedPosition()).PS) {name=PreSkinnedLocal, nodeColor=<rgba 0 0.8359 1 1>}
 ]}
