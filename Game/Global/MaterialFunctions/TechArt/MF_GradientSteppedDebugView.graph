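# Material function exposed to the library that visualizes a scalar gradient as a color debug view with repeating bands.
# The input, saturated by default through a static switch, drives a hue shift of pure red, so value maps to hue.
# The same value times a floored step count, ten by default, passes through Frac to form a sawtooth that multiplies the color, darkening the start of each band.
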MaterialFunction /Game/Global/MaterialFunctions/TechArt/MF_GradientSteppedDebugView
graph {
 properties {bExposeToLibrary=true}
 nodes [
  n1 = FunctionInput(Preview=MaterialFunctionCall() {materialFunction=&LinearGradient}.UGradient) {inputName=Gradient Input, inputType=FunctionInput_Scalar}
  n2 = FunctionInput(Preview=StaticBool() {value=true}) {bUsePreviewValueAsDefault=true, inputName=Clamp Input, inputType=FunctionInput_StaticBool, previewValue=<1 0 0 1>}
  n3 = StaticSwitch(True=Saturate(None=@n1), False=@n1, Value=@n2)
  n4 = Constant3Vector() {constant=<rgba 1 0 0 0>}
  n5 = FunctionInput() {bUsePreviewValueAsDefault=true, inputName=Num Steps, inputType=FunctionInput_Scalar, sortPriority=1000, previewValue=<10 0 0 1>}
  n6 = Multiply(A=@n3, B=Floor(None=@n5)) {constB=7}
  FunctionOutput(None=Multiply(
    A=MaterialFunctionCall(Hue Shift Percentage=@n3, Texture=@n4) {materialFunction=&HueShift}.Result
    B=Frac(None=@n6)
   ))
 ]
}
