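# Passthrough material function for hologram masking in the FE, meant to take the red channel of SkinFX_Mask.
# Input R feeds both A and B of a Lerp to R Out, so the value passes through unchanged.
# The Lerp alpha only keeps the UseSkinFXMaskForOpacity scalar parameter in the graph, flagging whether that red channel counts toward hologram masking.
# The function itself applies no masking.
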MaterialFunction /Game/Global/MaterialFunctions/TechArt/MF_MaskedGeoHoloControl
graph {
 properties {description="Passthrough MF for Hologram Complaint, must Input R from SkinFX_Mask"}
 nodes [
  n1 = FunctionInput() {inputName=R, inputType=FunctionInput_Scalar}
  n2 = ScalarParameter() {
   desc=Set to 1 if the Red Channel in the FX Mask should be counted for Hologram masking in the FE
   group=HologramMasking
   parameterName=UseSkinFXMaskForOpacity
  }
  FunctionOutput(None=LinearInterpolate(A=@n1, B=@n1, Alpha=@n2)) {outputName=R Out}
 ]
}
