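# Material function in the TechArt library that converts UV coordinates into polar coordinates around the (0.5, 0.5) center.
# It offsets the UVs by -0.5, then derives radius from the length of the vector scaled by LenghtScale.
# Angle comes from Arctangent2Fast divided by Pi times RadialScale.
# The output packs radius and angle as a two-component vector; a standard Arctangent2 node is left unused.
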MaterialFunction /Game/Global/MaterialFunctions/TechArt/MF_PolarCoordinates
graph {
 properties {bExposeToLibrary=true, libraryCategoriesText [TechArt]}
 nodes [
  n1 = Constant() {r=-0.5}
  n2 = FunctionInput(Preview=TextureCoordinate()) {bUsePreviewValueAsDefault=true, inputName=UV Coord, inputType=FunctionInput_Vector2, previewValue=<1 1 0 1>}
  n3 = Add(A=AppendVector(A=@n1, B=@n1), B=@n2)
  n4 = ComponentMask(None=@n3) {g=true}
  n5 = ComponentMask(None=@n3) {r=true}
  Arctangent2!(Y=@n4, X=@n5)
  n6 = FunctionInput() {bUsePreviewValueAsDefault=true, inputName=LenghtScale, inputType=FunctionInput_Scalar, previewValue=<1 0 0 0>}
  n7 = Multiply(A=@n6, B=@n5)
  n8 = Multiply(A=@n6, B=@n4)
  n9 = Add(A=Multiply(A=@n7, B=@n7), B=Multiply(A=@n8, B=@n8))
  n10 = Arctangent2Fast(Y=@n4, X=@n5)
  n11 = FunctionInput() {
   bUsePreviewValueAsDefault=true
   desc=Radial Gradient
   inputName=RadialScale
   inputType=FunctionInput_Scalar
   sortPriority=1
   previewValue=<1 0 0 0>
  }
  n12 = MaterialFunctionCall(Multiplier=@n11) {desc=Radial Gradient, materialFunction=&Pi}
  FunctionOutput(None=AppendVector(A=SquareRoot(None=@n9), B=Divide(A=@n10, B=@n12.Result)))
 ]
}
