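# Material function that layers a thin-film gradient sheen onto incoming material attributes, built for the Snow Knight character.
# A warped, scaled, offset Fresnel term samples one row of a gradient texture, which is added to a parameter base color by a Fresnel power mask, optionally weighted by roughness.
# It also adjusts roughness and metallic, and flattens normals on mobile quality tiers.
# A channel-selectable Alpha input blends the result over the original attributes.
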MaterialFunction /Game/Global/MaterialFunctions/TechArt/MF_ThinFilm_SnowKnight
graph {nodes [
  n1 = NamedRerouteUsage()
  n2 = VectorParameter() {parameterName=ThinFilmBaseColor, defaultValue=<rgba 0.2 0.2 0.2 1>}
  n3 = Constant() {r=1}
  n4 = Saturate(None=FunctionInput(Preview=Fresnel(ExponentIn=@n3) {baseReflectFraction=0}) {inputName=Fresnel, inputType=FunctionInput_Scalar, sortPriority=2})
  n5 = Multiply(
   A=ScalarParameter() {defaultValue=1, group=SK_Thin Film, parameterName=SKThinFilmScale, sortPriority=5}
   B=Power(Base=@n4, Exp=ScalarParameter() {defaultValue=0.5, group=SK_Thin Film, parameterName=SKThinFilmWarp, sortPriority=4})
  )
  n6 = ScalarParameter() {group=SK_Thin Film, parameterName=SKThinFilmOffset}
  n7 = TextureSample(UVs=MaterialFunctionCall(X=Add(A=@n5, B=@n6), Y=Constant() {r=0.25}) {materialFunction=&MakeFloat2}.Result) {
   samplerSource=SSM_Wrap_WorldGroupSettings
   texture=/Game/Characters/Player/Female/Medium/Bodies/F_MED_SnowKnight_Helm/Materials/AdvanceMaterials/T_SnowKnight_Gradients
  }
  n8 = Multiply(A=@n7.RGB, B=ScalarParameter() {defaultValue=0.5, group=SK_Thin Film, parameterName=SKThinFilmIntensity, sortPriority=2})
  n9 = Power(Base=@n4, Exp=ScalarParameter() {defaultValue=2, group=SK_Thin Film, parameterName=SKThinFilmExponent, sortPriority=3})
  n10 = GetMaterialAttributes(None=FunctionInput() {inputName=Material Attributes, inputType=FunctionInput_MaterialAttributes, sortPriority=1}) {attributeGetTypes [#1, #2, #3]}
  n11 = LinearInterpolate(
   A=@n9
   B=Multiply(A=LinearInterpolate(Alpha=@n10.Roughness) {constA=1, constB=0.01}, B=@n9)
   Alpha=ScalarParameter() {group=SK_Thin Film, parameterName=SKRoughnessInfluence}
  )
  n12 = ScalarParameter() {parameterName=SKThinFilmRoughnessBias}
  n13 = NamedRerouteUsage()
  n14 = MaterialFunctionCall(Normal=@n13, Flatness=Constant() {r=0.5}) {materialFunction=&FlattenNormal}
  n15 = SetMaterialAttributes!(
   MaterialAttributes=@n1
   BaseColor=LinearInterpolate(A=@n2.RGB, B=Add(A=@n2.RGB, B=@n8), Alpha=Saturate(None=@n11))
   Roughness=Saturate(None=Add(A=@n10.Roughness, B=@n12))
   Metallic=LinearInterpolate(
    A=@n10.Metallic
    B=ScalarParameter() {defaultValue=0.5, parameterName=ThinFilmMetallic}
    Alpha=ScalarParameter() {parameterName=UseThinFilmMetallic}
   )
   Normal=MaterialFunctionCall!(
    Low _ PC Low and Med=@n13
    "High _ XB1, PS4"=@n13
    "Epic _ PC High, XSX, PS5"=@n13
    --------
    "Mobile Low _ PC Perf, Android"=@n14.Result
    Mobile Med _ Switch=@n14.Result
   ) {materialFunction=&/MF_Fortnite_QualitySwitch_Vector3}.Result
  ) {attributeSetTypes [
    #4
    #1
    #2
    #3
   ]}
  n16 = FunctionInput() {bUsePreviewValueAsDefault=true, inputName=Alpha, sortPriority=3}
  FunctionOutput(None=BlendMaterialAttributes(A=@n1, B=@n15, Alpha=ChannelMaskParameter(None=@n16) {group=SK_Thin Film, parameterName=SKThinFilmMaskChannel}.R) {vertexAttributeBlendType=UseA})
  NamedRerouteDeclaration(None=@n10.MaterialAttributes) {name=Material Attributes, nodeColor=<rgba 0.0781 0 1 1>}
  NamedRerouteDeclaration(None=@n10.Normal) {name=Normal, nodeColor=<rgba 0 0.1094 1 1>}
 ]}
