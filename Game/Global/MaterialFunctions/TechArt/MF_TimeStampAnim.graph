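# Material function in the Anim library category that turns a start time and a length into a linear 0-1 animation value.
# Inputs are TimeStart, AnimLength and TimeCurrent, which uses the material Time node as its preview default.
# A time difference is divided and saturated to produce AnimValue, clamping it outside the animation window.
# Named reroutes carry values between the steps.
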MaterialFunction /Game/Global/MaterialFunctions/TechArt/MF_TimeStampAnim
graph {
 properties {bExposeToLibrary=true, libraryCategoriesText [Anim]}
 nodes [
  FunctionOutput(None=NamedRerouteUsage()) {description=0-1 linear value based on time and length, outputName=AnimValue}
  n1 = FunctionInput() {bUsePreviewValueAsDefault=true, inputName=TimeStart, inputType=FunctionInput_Scalar, previewValue=<-1 0 0 1>}
  NamedRerouteDeclaration(None=@n1) {name=TimeStart_Route, nodeColor=<rgba 1 0 0.3047 1>}
  n2 = FunctionInput(Preview=Time()) {bUsePreviewValueAsDefault=true, inputName=TimeCurrent, inputType=FunctionInput_Scalar, sortPriority=2, previewValue=<-1 0 0 1>}
  NamedRerouteDeclaration(None=@n2) {name=TimeCurrent_Route, nodeColor=<rgba 0 0.625 1 1>}
  n3 = FunctionInput() {bUsePreviewValueAsDefault=true, inputName=AnimLength, inputType=FunctionInput_Scalar, sortPriority=1, previewValue=<1 0 0 1>}
  NamedRerouteDeclaration(None=@n3) {name=AnimLength_Route, nodeColor=<rgba 0.9453 1 0 1>}
  n4 = Subtract(A=NamedRerouteUsage(), B=NamedRerouteUsage())
  n5 = NamedRerouteUsage()
  NamedRerouteDeclaration(None=Saturate(None=Divide(A=@n4, B=@n5))) {name=CurAnimValue_Route, nodeColor=<rgba 0.7344 0 1 1>}
 ]
}
