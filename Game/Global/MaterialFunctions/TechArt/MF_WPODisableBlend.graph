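# Material function that outputs a 0-1 ramp, reaching 1 at the WPO disable distance, for fading world position offset without popping.
# Custom HLSL reads the primitive's squared disable distance, and camera-to-instance-origin distance is remapped against its root.
# A static switch starts the ramp FadeBlendDistance earlier, shortened for small disable distances, or uses that input directly as start.
# Pixel shader use goes through a vertex interpolator.
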MaterialFunction /Game/Global/MaterialFunctions/TechArt/MF_WPODisableBlend
graph {
 properties {
  bExposeToLibrary=true
  description=Provides a 0-1 value to fade WPO before WPODisableDistance is reached to avoid popping when WPO turns off.WARNING: Will use a vertex interpolator if referenced in pixel shader
  libraryCategoriesText [Fortnite]
 }
 nodes [
  n1 = Custom!(None) {
   code="if ((Parameters.SceneData.Primitive.Flags && PRIMITIVE_SCENE_DATA_FLAG_WPO_DISABLE_DISTANCE) != 0)\r\n{\r\n\treturn Parameters.SceneData.Primitive.InstanceWPODisableDistanceSquared;\r\n}\r\nelse\r\n{\r\n\treturn 1e+38f;\r\n}"
   description=GetWPODisableDistance
  }
  n2 = SquareRoot(None=ComponentMask(None=@n1) {r=true})
  n3 = FunctionInput() {
   bUsePreviewValueAsDefault=true
   description=Distance before WPODisableDistance that we begin to blend out
   inputName=FadeBlendDistance
   inputType=FunctionInput_Scalar
   previewValue=<500 0 0 1>
  }
  n4 = Subtract(A=@n2, B=If(A=@n2, B=@n3, "A > B"=@n3, "A < B"=Multiply(A=@n2) {constB=0.25}))
  n5 = FunctionInput(Preview=StaticBool() {value=true}) {
   bUsePreviewValueAsDefault=true
   description="Should we gracefully handle if the fade distance is less than the WPO Disable Distance? (Maybe set to False to save instructions if you are handling distance very carefully)"
   inputName=AutoHandleBadDistance
   inputType=FunctionInput_StaticBool
   sortPriority=1
   previewValue=<1 0 0 1>
  }
  n6 = TransformPosition(None=Constant3Vector()) {transformSourceType=TRANSFORMPOSSOURCE_Instance, transformType=TRANSFORMPOSSOURCE_World}
  n7 = CameraPositionWS()
  n8 = Saturate(None=InverseLinearInterpolate(
    A=StaticSwitch(True=@n4, False=@n3, Value=@n5) {desc="Handle situations where the blend distance is less than the Disable Distance (Would normally pop if not)"}
    B=@n2
    Value=Distance(A=@n6, B=@n7)
   ))
  n9 = VertexInterpolator(VS=@n8)
  FunctionOutput(None=ShaderStageSwitch(PixelShader=ComponentMask(None=@n9.PS) {r=true}, VertexShader=@n8)) {
   description="Provides a 0-1 value to fade WPO before WPODisableDistance is reached to avoid popping when WPO turns off.\r\n\r\nWARNING: Will use a vertex interpolator if referenced in pixel shader"
   outputName=VertexShaderOutput
  }
 ]
}
